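# One-shot stereo MetaSound for the additive step layer of cosmetic Kicks footsteps.
# A Stance check either ends the sound or starts a delayed random pick from LowIntensity.Sounds, with randomized pitch and volume scaled by KickAdditiveGainFade.
# A Character.Affiliation check at playback routes output either dry or through the spatialization patch and a third-person footstep distance volume curve.
# Concurrency is capped at two voices.
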MetaSoundSource /FortMetasound/Cosmetics/Kicks/MSS_Kicks_Additive_Step
inputs [
 Character.Affiliation {type=Enum:FortAudioAffiliation}
 Character.Health {type=Float}
 Character.UniqueID {type=Int32}
 Delay Time {type=Time}
 KickAdditiveGainFade {type=Float}
 LowIntensity.Sounds {type=WaveAsset:Array}
 Random.Pitch.Base {type=Float}
 Random.Pitch.Spread {type=Float}
 Random.Volume.Base {type=Float}
 Random.Volume.Spread {type=Float}
 Stance {type=Int32}
 UE.Attenuation.Distance {type=Float}
 UE.Source.OnPlay {type=Trigger}
 UE.Spatialization.Azimuth {type=Float}
 UE.Spatialization.Elevation {type=Float}
]
interfaces [Fort.Character:1.0, UE.Attenuation:1.0, UE.OutputFormat.Stereo:1.0, UE.Source.OneShot:1.0, UE.Source:1.1, UE.Spatialization:1.0]
outputs [
 "UE.OutputFormat.Stereo.Audio:0" {type=Audio}
 "UE.OutputFormat.Stereo.Audio:1" {type=Audio}
 UE.Source.OneShot.OnFinished {type=Trigger}
]
pages [Default {
  input_defaults {
   Character.Affiliation=0
   Character.Health=0
   Character.UniqueID=0
   Delay Time=0
   KickAdditiveGainFade=1
   Random.Pitch.Base=0
   Random.Pitch.Spread=0
   Random.Volume.Base=0
   Random.Volume.Spread=0
   Stance=1
   UE.Attenuation.Distance=0
   UE.Source.OnPlay=null
   UE.Spatialization.Azimuth=0
   UE.Spatialization.Elevation=0
   LowIntensity.Sounds []
  }
  nodes [
   InitVariable.Audio()
   InitVariable.Bool()
   n1 = InitVariable.Trigger()
   n2 = TriggerCompare.Clamp:Int32(A=^Stance, Compare=^UE.Source.OnPlay) {also {defaults {B=2}}}
   n3 = InitVariable.Trigger()
   TriggerDelay_0 = UE.Trigger Delay(Delay Time=^Delay Time, In=@n2.True)
   ArrayRandomGet_0 = Array.Random Get:WaveAsset:Array(In Array=^LowIntensity.Sounds, Next=@TriggerDelay_0.Out) {also {defaults {No Repeats=-1, Weights []}}}
   n4 = AbsoluteValue.Abs:Float(Input=^Random.Pitch.Spread)
   WavePlayer_0 = UE.Wave Player:Stereo(
    Play=@ArrayRandomGet_0.On Next
    Wave Asset=@ArrayRandomGet_0.Value
    Pitch Shift=UE.RandomFloat(
     Next=^UE.Source.OnPlay
     Max=UE.Add:Float(PrimaryOperand=^Random.Pitch.Base, AdditionalOperands=@n4.Value).Out
     Min=UE.Subtract:Float(PrimaryOperand=^Random.Pitch.Base, AdditionalOperands=@n4.Value).Out
    ).Value
   ) {also {version [1,1]}}
   TriggerAnyTriggerAny2_0 = "TriggerAny.Trigger Any (2)"(
    In 0=VariableAccessor.Trigger(Variable=VariableMutator.Trigger(Variable=@n1.Variable, Value=@n2.False).Variable).Value
    In 1=VariableAccessor.Trigger(Variable=VariableMutator.Trigger(Variable=@n3.Variable, Value=@WavePlayer_0.On Finished).Variable).Value
   )
   n5 = &TriggerCompareAffiliation(
    A=^Character.Affiliation
    B=VariableAccessor.Enum:FortAudioAffiliation(Variable=InitVariable.Enum:FortAudioAffiliation().Variable).Value
    Compare=@WavePlayer_0.On Play
   )
   n6 = InitVariable.Audio()
   n7 = VariableAccessor.Trigger(Variable=VariableMutator.Trigger(Variable=InitVariable.Trigger().Variable, Value=@n5.True).Variable)
   n8 = VariableAccessor.Trigger(Variable=VariableMutator.Trigger(Variable=InitVariable.Trigger().Variable, Value=@n5.False).Variable)
   TriggerRouteTriggerRouteFloat2_0 = "TriggerRoute.Trigger Route (Float, 2):Float"(Value 0=^UE.Attenuation.Distance, Set 1=@n7.Value, Set 0=@n8.Value) {also {defaults {Value 1=0}}}
   TriggerRouteTriggerRouteFloat2_1 = "TriggerRoute.Trigger Route (Float, 2):Float"(Value 0=^UE.Spatialization.Azimuth, Set 1=@n7.Value, Set 0=@n8.Value)
   n9 = AbsoluteValue.Abs:Float(Input=^Random.Volume.Spread)
   DecibelstoLinearGain_0 = UE.Decibels to Linear Gain:Float(Decibels=UE.RandomFloat(
     Next=^UE.Source.OnPlay
     Max=UE.Add:Float(PrimaryOperand=^Random.Volume.Base, AdditionalOperands=@n9.Value).Out
     Min=UE.Subtract:Float(PrimaryOperand=^Random.Volume.Base, AdditionalOperands=@n9.Value).Out
    ).Value)
   n10 = UE.Multiply:Float(AdditionalOperands=^KickAdditiveGainFade, PrimaryOperand=@DecibelstoLinearGain_0.Linear Gain)
   n11 = VariableAccessor.Audio(Variable=VariableMutator.Audio(
     Variable=InitVariable.Audio().Variable
     Value=UE.Multiply:Audio by Float(AdditionalOperands=@n10.Out, PrimaryOperand=@WavePlayer_0.Out Right).Out
    ).Variable)
   n12 = VariableAccessor.Audio(Variable=VariableMutator.Audio(
     Variable=InitVariable.Audio().Variable
     Value=UE.Multiply:Audio by Float(AdditionalOperands=@n10.Out, PrimaryOperand=@WavePlayer_0.Out Left).Out
    ).Variable)
   n13 = /FortMetasound/Attenuation/MSP_Attenuation_Spatialization(
    Distance=@TriggerRouteTriggerRouteFloat2_0.Value
    Azimuth=@TriggerRouteTriggerRouteFloat2_1.Value
    Stereo.InR=VariableAccessor.Audio(Variable=@n11.Variable).Value
    Stereo.InL=VariableAccessor.Audio(Variable=@n12.Variable).Value
   ) {also {defaults {StereoSpread=0}}}
   TriggerRouteTriggerRouteBool2_0 = "TriggerRoute.Trigger Route (Bool, 2):Bool"(Set 1=@n7.Value, Set 0=@n8.Value) {also {defaults {Value 0="<MetaSoundBuilderResult.SUCCEEDED: 0>"}}}
   n14 = &MSP_Attenuation_Volume_Curve(Distance=^UE.Attenuation.Distance, bEnable=@TriggerRouteTriggerRouteBool2_0.Value) {also {defaults {AttenuationCurves=/Game/Sounds/Fort_Player/Footsteps/Settings/WTB_FS_3pCurve, FalloffDistance=3584}}}
   n15 = UE.Multiply:Audio by Float(PrimaryOperand=@n13.Right, AdditionalOperands=@n14.VolumeMultiplier)
   TriggerRouteTriggerRouteAudio2_0 = "TriggerRoute.Trigger Route (Audio, 2):Audio"(
    Set 0=UE.Reroute(Value=@n5.True).Value
    Set 1=UE.Reroute(Value=@n5.False).Value
    Value 1=VariableAccessor.Audio(Variable=VariableMutator.Audio(Variable=@n6.Variable, Value=@n15.Out).Variable).Value
    Value 0=@n11.Value
   )
   n16 = InitVariable.Audio()
   n17 = UE.Multiply:Audio by Float(PrimaryOperand=@n13.Left, AdditionalOperands=@n14.VolumeMultiplier)
   TriggerRouteTriggerRouteAudio2_1 = "TriggerRoute.Trigger Route (Audio, 2):Audio"(
    Set 0=UE.Reroute(Value=@n5.True).Value
    Set 1=UE.Reroute(Value=@n5.False).Value
    Value 1=VariableAccessor.Audio(Variable=VariableMutator.Audio(Variable=@n16.Variable, Value=@n17.Out).Variable).Value
    Value 0=@n12.Value
   )
  ]
  links [
   {dst="@UE.Source.OneShot.OnFinished", src=TriggerAnyTriggerAny2_0.Out}
   {dst="@UE.OutputFormat.Stereo.Audio:1", src=TriggerRouteTriggerRouteAudio2_0.Value}
   {dst="@UE.OutputFormat.Stereo.Audio:0", src=TriggerRouteTriggerRouteAudio2_1.Value}
  ]
 }]
properties {
 assetImportData=&self:AssetImportData
 attenuationSettings=/Game/Sounds/Fort_Player/Footsteps/Settings/Fort_Player_FS_1P_Att
 audioPropertiesSheet=/BRCosmetics/Sounds/Settings/PropertySheets/Kicks/APS_Cosmetic_Kicks_Step
 bOverrideConcurrency=true
 numChannels=2
 outputFormat=Stereo
 sampleRate=48000
 soundClassObject=&Fort_Player_Foley
 concurrencyOverrides {maxCount=2}
}
