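# MetaSound Source preset that plays the additive step foley layer for the LabelPeelCrinkle Kicks cosmetic.
# It inherits its graph from MSS_Kicks_Additive_Step and supplies ten low-intensity crinkle waves, a short delay time and a slight volume spread without pitch variation.
# The stereo one-shot uses first-person footstep attenuation, the player foley sound class, the Kicks step property sheet and a tight concurrency override.
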MetaSoundSource /BRCosmetics/Sounds/Kicks/LabelPeelCrinkle/MSS_LabelPeelCrinkle_Additive_Step_Preset
preset_parent=/FortMetasound/Cosmetics/Kicks/MSS_Kicks_Additive_Step
inputs [
 Character.Affiliation {type=Enum:FortAudioAffiliation}
 Character.Health {type=Float}
 Character.UniqueID {type=Int32}
 Delay Time {type=Time}
 KickAdditiveGainFade {type=Float}
 LowIntensity.Sounds {type=WaveAsset:Array}
 Random.Pitch.Base {type=Float}
 Random.Pitch.Spread {type=Float}
 Random.Volume.Base {type=Float}
 Random.Volume.Spread {type=Float}
 Stance {type=Int32}
 UE.Attenuation.Distance {type=Float}
 UE.Source.OnPlay {type=Trigger}
 UE.Spatialization.Azimuth {type=Float}
 UE.Spatialization.Elevation {type=Float}
]
interfaces [Fort.Character:1.0, UE.Attenuation:1.0, UE.OutputFormat.Stereo:1.0, UE.Source.OneShot:1.0, UE.Source:1.1, UE.Spatialization:1.0]
outputs [
 "UE.OutputFormat.Stereo.Audio:0" {type=Audio}
 "UE.OutputFormat.Stereo.Audio:1" {type=Audio}
 UE.Source.OneShot.OnFinished {type=Trigger}
]
pages [Default {input_defaults {
   Character.Affiliation=0
   Character.Health=0
   Character.UniqueID=0
   Delay Time=0.1
   KickAdditiveGainFade=1
   Random.Pitch.Base=0
   Random.Pitch.Spread=0
   Random.Volume.Base=0
   Random.Volume.Spread=-2
   Stance=1
   UE.Attenuation.Distance=0
   UE.Source.OnPlay=null
   UE.Spatialization.Azimuth=0
   UE.Spatialization.Elevation=0
   LowIntensity.Sounds [
    /BRCosmetics/Sounds/Kicks/LabelPeelCrinkle/WaveAssets/IP_LabelPeelCrinkle_Additive_Foley_LowIntensity_01
    /BRCosmetics/Sounds/Kicks/LabelPeelCrinkle/WaveAssets/IP_LabelPeelCrinkle_Additive_Foley_LowIntensity_02
    /BRCosmetics/Sounds/Kicks/LabelPeelCrinkle/WaveAssets/IP_LabelPeelCrinkle_Additive_Foley_LowIntensity_03
    /BRCosmetics/Sounds/Kicks/LabelPeelCrinkle/WaveAssets/IP_LabelPeelCrinkle_Additive_Foley_LowIntensity_04
    /BRCosmetics/Sounds/Kicks/LabelPeelCrinkle/WaveAssets/IP_LabelPeelCrinkle_Additive_Foley_LowIntensity_05
    /BRCosmetics/Sounds/Kicks/LabelPeelCrinkle/WaveAssets/IP_LabelPeelCrinkle_Additive_Foley_LowIntensity_06
    /BRCosmetics/Sounds/Kicks/LabelPeelCrinkle/WaveAssets/IP_LabelPeelCrinkle_Additive_Foley_LowIntensity_07
    /BRCosmetics/Sounds/Kicks/LabelPeelCrinkle/WaveAssets/IP_LabelPeelCrinkle_Additive_Foley_LowIntensity_08
    /BRCosmetics/Sounds/Kicks/LabelPeelCrinkle/WaveAssets/IP_LabelPeelCrinkle_Additive_Foley_LowIntensity_09
    /BRCosmetics/Sounds/Kicks/LabelPeelCrinkle/WaveAssets/IP_LabelPeelCrinkle_Additive_Foley_LowIntensity_10
   ]
  }}]
properties {
 assetImportData=&self:AssetImportData
 attenuationSettings=/Game/Sounds/Fort_Player/Footsteps/Settings/Fort_Player_FS_1P_Att
 audioPropertiesSheet=/BRCosmetics/Sounds/Settings/PropertySheets/Kicks/APS_Cosmetic_Kicks_Step
 bOverrideConcurrency=true
 numChannels=2
 outputFormat=Stereo
 sampleRate=48000
 soundClassObject=&Fort_Player_Foley
 concurrencyOverrides {maxCount=2}
}
